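# Translucent, unlit, two-sided trail material for the RoseScour skydiving contrail, flagged for Niagara mesh particles.
# A panning, swizzled-UV pattern texture is thresholded and combined with particle color for emissive, while UV gradients build a soft-edged mask from the same texture.
# World position offset sums a Niagara-driven bend relative to particle position, a view-space sine wobble with optional near-camera damping, and switchable SplineThicken width with tip and base gradients.
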Material /BRCosmetics/Athena/Effects/Particles/Skydiving/NS_Contrail_RoseScour/M_RoseScour_Trail
graph {
 properties {
  bUsedWithMeshDeformer=true
  bUsedWithNanite=true
  bUsedWithNiagaraMeshParticles=true
  bUsedWithSkeletalMesh=true
  blendMode=BLEND_Translucent
  pixelDepthOffsetMode=PDOM_Legacy
  shadingModel=MSM_Unlit
  twoSided=true
  usedShadingModels=Unlit
 }
 nodes [
  ComponentMask(None=OneMinus(None=TextureCoordinate())) {g=true}
  Constant() {r=1}
  n1 = TextureSample(UVs=Panner(Coordinate=MaterialFunctionCall() {materialFunction=&Swizzle}.YX) {speedX=0.8}) {texture=&/T_RoseScour_Pattern}
  n2 = Step(Y=@n1.R) {constX=0.5}
  n3 = OneMinus(None=@n2)
  n4 = ParticleColor()
  n5 = Multiply(A=@n2, B=@n1.R)
  MaterialFunctionCall(Emissive=Add(A=Multiply(A=@n3) {constB=0.5}, B=Add(A=@n4.RGB, B=@n5))) {materialFunction=&MF_FX_EmissiveAdjustments}
  n6 = ComponentMask(None=TextureCoordinate()) {g=true}
  n7 = Add(A=@n1.R, B=Multiply(A=Multiply(A=@n1.B) {constB=0.5}, B=OneMinus(None=@n6)))
  n8 = TextureCoordinate()
  n9 = ComponentMask(None=@n8) {g=true}
  n10 = Power(Base=Multiply(A=OneMinus(None=@n9)) {constB=5})
  n11 = ComponentMask(None=@n8) {r=true}
  n12 = Multiply(A=@n11, B=OneMinus(None=@n11))
  Multiply(A=Multiply(A=Multiply(A=@n7, B=@n9), B=Saturate(None=@n10)), B=Power(Base=Multiply(A=@n12) {constB=5}))
  n13 = DynamicParameter() {desc=Parameters set by Niagara, parameterIndex=2, paramNames [X,Y,Z,WPOIntensity]}
  n14 = Subtract(
   A=MaterialFunctionCall(X=@n13.X, Y=@n13.X, Z=@n13.X) {materialFunction=&MakeFloat3}.Result
   B=ParticlePositionWS()
  )
  n15 = ScalarParameter() {defaultValue=-1000, group=WPO, parameterName=InterpolateWPO_Clamp_min}
  n16 = ScalarParameter() {defaultValue=1000, group=WPO, parameterName=InterpolateWPO_Clamp_max}
  n17 = Power(Base=OneMinus(None=VertexColor().R), Exp=ScalarParameter() {defaultValue=3, group=WPO, parameterName=InterpolateWPO_BendPower})
  n18 = VectorParameter() {group=WPO, parameterName=WPO_Direction, defaultValue=<rgba 0.25 0.25 0 1>}
  n19 = TextureCoordinate() {uTiling=0, vTiling=0.7}
  n20 = ComponentMask(None=VectorParameter() {group=WPO, parameterName=WPO_UV_Scale, defaultValue=<rgba 1 1 0 1>}.RGB) {g=true, r=true}
  n21 = Add(A=ComponentMask(None=Multiply(A=@n19, B=@n20)) {g=true}, B=ScalarParameter() {group=WPO, parameterName=WPO_Offset})
  n22 = VertexColor()
  n23 = Add(
   A=Add(
    A=Add(A=@n21, B=@n22.G)
    B=DynamicParameter() {defaultValue=<rgba 1 1 1 0>, paramNames [RandomOffset, WPO_Intensity, WPO_Speed, Dissolve]}.RandomOffset
   )
   B=VertexColor().R
  )
  n24 = Multiply(A=Time(), B=DynamicParameter() {defaultValue=<rgba 1 1 1 0>, paramNames [RandomOffset, WPO_Intensity, WPO_Speed, Dissolve]}.WPO_Speed) {constB=2.8}
  n25 = Multiply(A=Sine(None=Add(A=@n23, B=@n24) {constB=0})) {constB=20}
  n26 = ComponentMask(None=TextureCoordinate()) {g=true}
  n27 = LinearInterpolate(Alpha=Add(A=OneMinus(None=@n26)) {constB=0.5})
  n28 = Transform(None=Multiply(A=Multiply(A=@n18.RGB, B=@n25), B=Saturate(None=@n27))) {transformSourceType=TRANSFORMSOURCE_View}
  n29 = ComponentMask(None=TextureCoordinate()) {g=true}
  n30 = Power(
   Base=Add(A=OneMinus(None=@n29), B=ScalarParameter() {group=WPO, parameterName=Bottom_Mask_Offset})
   Exp=ScalarParameter() {defaultValue=2.9, group=WPO, parameterName=Power}
  )
  n31 = ScalarParameter() {defaultValue=50, group=WPO, parameterName=Power_Mask_Edge}
  n32 = Multiply(
   A=Multiply(
    A=Multiply(A=@n28) {constB=0.5}
    B=DynamicParameter() {defaultValue=<rgba 1 1 1 0>, paramNames [RandomOffset, WPO_Intensity, WPO_Speed, Dissolve]}.WPO_Intensity
   )
   B=Saturate(None=Multiply(A=@n30, B=@n31) {constB=50})
  )
  n33 = ScalarParameter() {defaultValue=0.3, group=WPO, parameterName=Min_Near_Camera_WPO}
  n34 = MaterialFunctionCall(Distance=ScalarParameter() {defaultValue=256, group=WPO, parameterName=Near_Camera_fade}) {materialFunction=&Near_Camera_Fade}
  n35 = LinearInterpolate(
   B=Saturate(None=LinearInterpolate(A=@n33, Alpha=@n34.Result) {constA=0.5})
   Alpha=ScalarParameter() {group=WPO, parameterName=Use_Near_Camera_Fade}
  ) {constA=1}
  n36 = Add(
   A=LinearInterpolate(B=Clamp(None=@n14, Min=@n15, Max=@n16) {maxDefault=1000, minDefault=-1000}, Alpha=Multiply(A=@n17, B=@n13.WPOIntensity))
   B=LinearInterpolate(B=Multiply(A=@n32, B=@n35), Alpha=ScalarParameter() {defaultValue=1, group=WPO, parameterName=Use_WPO})
  )
  n37 = ComponentMask(None=LinearInterpolate(
    A=TextureCoordinate()
    B=TextureCoordinate() {coordinateIndex=1}
    Alpha=ScalarParameter() {defaultValue=1, group=NitroFlow_Options, parameterName=NitroFlow_UseUV1}
   )) {g=true}
  n38 = Multiply(
   A=Power(
    Base=Multiply(A=@n37) {constB=2}
    Exp=ScalarParameter() {defaultValue=1, group=NitroFlow_Options, parameterName=Tip_GradientPower, sortPriority=3}
   ) {constExponent=1}
   B=ScalarParameter() {defaultValue=1, group=NitroFlow_Options, parameterName=Tip_GradientMultiply, sortPriority=2}
  )
  n39 = OneMinus(None=@n37)
  n40 = Multiply(
   A=Power(
    Base=Multiply(A=@n39) {constB=2}
    Exp=ScalarParameter() {defaultValue=1, group=NitroFlow_Options, parameterName=Base_GradientPower, sortPriority=4}
   ) {constExponent=1}
   B=ScalarParameter() {defaultValue=5, group=NitroFlow_Options, parameterName=Base_GradientMultiply, sortPriority=5}
  )
  n41 = Multiply(
   A=MaterialFunctionCall(
    WidthBase=ScalarParameter() {defaultValue=75, group=NitroFlow_Options, parameterName=BaseWidth, sortPriority=0}
    WidthTip=ScalarParameter() {defaultValue=75, group=NitroFlow_Options, parameterName=TipWidth, sortPriority=1}
   ) {materialFunction=&SplineThicken}.WorldPositionOffset
   B=Multiply(A=Saturate(None=@n38), B=Saturate(None=@n40))
  )
  n42 = ScalarParameter() {defaultValue=1, group=NitroFlow_Options, parameterName=UseSplineThicken, sortPriority=-1}
  Reroute(None=Add(A=@n36, B=LinearInterpolate(A=@n36, B=@n41, Alpha=@n42)))
 ]
}
