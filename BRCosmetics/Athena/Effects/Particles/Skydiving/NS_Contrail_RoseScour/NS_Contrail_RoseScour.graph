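# Niagara skydiving contrail cosmetic RoseScour, drawing warm orange trails and effects from sockets on the character's skeletal mesh.
# Four mesh emitters layer soft, pattern, chain and red trails, alongside sprite sparks and flickering hand-and-feet flares with lights.
# System dive, glide-lean, fade and front-end preview values drive scale, offset, color and orientation.
# Smoothed particle positions feed trail materials, and the soft trail tints by time of day.
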NiagaraSystem /BRCosmetics/Athena/Effects/Particles/Skydiving/NS_Contrail_RoseScour/NS_Contrail_RoseScour
dependencies {
 usedDataInterfaces [
  &NiagaraDataInterfaceArrayFloat3
  &NiagaraDataInterfaceArrayInt32
  &NiagaraDataInterfaceCurve
  &NiagaraDataInterfacePlatformSet
  &NiagaraDataInterfaceArrayFloat
  &NiagaraDataInterfaceSkeletalMesh
 ]
 usedDynamicInputs [/Niagara/DynamicInputs/Lerp/SmoothLerpOverTime_Float]
 usedModules [
  &self:SystemSpawnScript.SystemScriptSource.SystemScriptGraph.NiagaraNodeAssignment_2.SetVariables_#1
  &self:SystemSpawnScript.SystemScriptSource.SystemScriptGraph.NiagaraNodeAssignment_3.SetVariables_#2
  &self:SystemSpawnScript.SystemScriptSource.SystemScriptGraph.NiagaraNodeAssignment_0.SetVariables_#3
  &self:SystemSpawnScript.SystemScriptSource.SystemScriptGraph.NiagaraNodeAssignment_1.SetVariables_#4
  &SystemState
  &self:SystemSpawnScript.SystemScriptSource.SystemScriptGraph.NiagaraNodeAssignment_4.SetVariables_#5
 ]
}
emitters {
 Hand_Feet_Flares {
  emitterUpdateScript [
   &EmitterState(
    Inactive Response=::NewEnumerator1
    Life Cycle Mode=::NewEnumerator1
    Loop Behavior=::NewEnumerator1
    Loop Duration Mode=::NewEnumerator1
    UseLoopDelay=false
   )
   /Niagara/Noise/Waveform(
    EnableExponent1=false
    GlobalClampMax=false
    GlobalClampMin=false
    UseInput1=false
    UsePhase1=false
    "Waveform [1]":ENiagara_Waveforms=::NewEnumerator12
    "[1] Frequency"=7
   )
   /SetVariables(Emitter.Flicker:NiagaraFloat=&Add_Float(A=&Multiply_Float(A=$Emitter.Waveform.Waveform, B=0.5), B=0.5)) {isSetParameters=true}
   &SpawnBurst_Instantaneous(Spawn Count=2, Use Spawn Probability=false, UseLoopCountLimit=false) {module=SpawnBurst_Instantaneous001}
  ]
  particleSpawnScript [
   &PartitionParticles(Number Of Partitions=&MergeHelper_IntSwitch(0=2, 1=4, 2=?, 3=?, 4=?, 5=?, Index=&SpawnGroup(), Num=::NewEnumerator0))
   &InitializeParticle(
    Color=<rgba 0.2 0.1473 0.1287 1.8>
    Lifetime=5
    Sprite Size Mode=::NewEnumerator3
    Uniform Sprite Size=100
    UsePositionOffset=false
    Write Material Random=true
   )
  ]
  particleUpdateScript [
   &ParticleState(Kill Particles When Lifetime Has Elapsed=false)
   &SkeletalMeshLocation(
    Bone / Socket Index=$Particles.PartitionParticles.Partition
    Bone Sampling Mode=::NewEnumerator4
    MaskSpawnGroup=false
    Sampled Position Offset=&Lerp_Vector(
     A=&MergeHelper_VectorSwitch(0=<0 0 0>, 1=<0 0 0>, 2=?, 3=?, Index=$Particles.PartitionParticles.Partition, Num=::NewEnumerator0)
     B=&MergeHelper_VectorSwitch(0=<0 0 0>, 1=<0 0 0>, 2=?, 3=?, Index=$Particles.PartitionParticles.Partition, Num=::NewEnumerator0)
     Alpha=$System.SmoothDiveAlpha
    )
    Skeletal Mesh=$System.Skeletal Mesh
   ) {module=SkeletalMeshLocation001}
   &ScaleColor(
    Scale Alpha=&Multiply_Float(A=0.75, B=$System.SystemFade)
    Scale RGB=&MakeVectorFromLinearColorRGB(Color=&Multiply_LinearColorByFloat(
      Linear Color=&LerpLinearColors(StartColor=<rgba 1.5 0.9822 0.3 0.2957>, EndColor=<rgba 3 1.6891 0 0.3478>, LerpFactor=$System.SmoothDiveAlpha)
      Float=$Emitter.Flicker
     ))
    ScaleA=true
    ScaleRGB=true
   )
   &DynamicMaterialParameters(
    Index 0 Param 1=5
    Index 0 Param 2=1
    Index 0 Param 4=0.3
    Module.Index0Param0Write=true
    Module.Index0Param1Write=true
    Module.Index0Param2Write=true
    Module.Index0Param3Write=true
   )
   &FN_Light_Attributes(
    Color=$Particles.Color
    Create Color Attribute=true
    Create Exponent Attribute=true
    Create Radius Attribute=true
    Exponent=5
    Radius=100
    Scale Color=&Multiply_Float(A=5, B=$Emitter.Flicker)
   )
   &FN_Sprite_Attributes(
    Color=&MultiplyLinearColor(
     Linear Color=<rgba 1.5 1.2188 1.125 1>
     Scale Factor=&MakeVector4FromVectorAndFloat(
      "Vector (RGB)"=&MakeVectorFromLinearColorRGB(Color=&Multiply_LinearColorByFloat(
        Linear Color=&LerpLinearColors(StartColor=<rgba 1.5 1.0145 0.375 0.2957>, EndColor=<rgba 3 1.9513 0.6 0.3478>, LerpFactor=$System.SmoothDiveAlpha)
        Float=$Emitter.Flicker
       ))
      "Float (Alpha)"=&Multiply_Float(A=0.25, B=$System.SystemFade)
     )
    )
    Create Color Attribute=true
    Module Output:ENiagara_AttributeSamplingApplyOutput=::NewEnumerator1
   )
   &CameraOffset(Camera Offset Amount=-5)
  ]
  rendererClasses [&NiagaraLightRendererProperties, &NiagaraSpriteRendererProperties]
  renderers [
   {rendererClass=&NiagaraLightRendererProperties, rendererIndex=0}
   {rendererClass=&NiagaraSpriteRendererProperties, rendererIndex=1}
   {rendererClass=&NiagaraSpriteRendererProperties, rendererIndex=2}
  ]
 }
 Red_Trail {
  emitterUpdateScript [
   &EmitterState(
    Inactive Response=::NewEnumerator1
    Life Cycle Mode=::NewEnumerator1
    Loop Behavior=::NewEnumerator1
    Loop Duration Mode=::NewEnumerator1
    UseLoopDelay=false
   )
   &SpawnBurst_Instantaneous(Spawn Count=4, Use Spawn Probability=false, UseLoopCountLimit=false)
  ]
  particleSpawnScript [
   &FN_SetSpawnGroup(Number of Spawn Groups=2)
   &InitializeParticle(
    Color=<rgba 1 0.8473 0.3986 0.5>
    Lifetime=0.5
    Mesh Scale=<0.35 0.35 2.5>
    Mesh Scale Mode=::NewEnumerator4
    UsePositionOffset=false
    Write Material Random=true
   )
   &InitialMeshOrientation(
    Orientation Axis=<0 0 1>
    Orientation Vector=<0 0 1>
    Rotation=&MergeHelper_VectorSwitch(0=<0 0 0.25>, 1=<0 0 0>, 2=?, 3=?, Index=$Particles.SpawnGroup, Num=::NewEnumerator0)
    Rotation Coordinate Space=::NewEnumerator0
    Use Orientation Vector=true
    Use Rotation Vector=true
   )
   &CameraOffset(Camera Offset Amount=8)
  ]
  particleUpdateScript [
   &ParticleState(Kill Particles When Lifetime Has Elapsed=false, Let Infinitely Lived Particles Die When Emitter Deactivates=true)
   &InterpolateOverTime(Rate Of Change=10, Target Vector Value:NiagaraPosition=$Particles.Position, Variable Type=::NewEnumerator2)
   /SetVariables(Particles.DY:Vector4f=&MakeFloat4FromVectorAndFloat("Vector (RGB)"=$Output.InterpolateOverTime.Moving Average, "Float (Alpha)"=1)) {isSetParameters=true}
   /SetVariables(Particles.DynamicMaterialParameter2:Vector4f=&MakeFloat4FromVectorAndFloat("Vector (RGB)"=$Output.InterpolateOverTime.Moving Average, "Float (Alpha)"=1)) {isSetParameters=true}
   &DynamicMaterialParameters(
    Index 0 Param 1=&Multiply_Float(
     A=&RandomRangeFloat(Minimum=0.7, Maximum=1.2)
     B=&SelectFloatFromArray(Direct Array Index=&SpawnGroup(), Float Selection Array=DataInterface() {propertyValues {FloatData [0.5,0.5]}})
    )
    Index 0 Param 2=&Multiply_Float(A=&RandomRangeFloat(Minimum=1, Maximum=1), B=1)
    Index 0 Param 3=&Multiply_Float(
     A=&Multiply_Float(
      A=&RandomRangeFloat(Minimum=1, Maximum=1.8)
      B=&Lerp_Float(A=1, B=1.2, Alpha=$Particles.Dive)
     )
     B=&SelectFloatFromArray(Direct Array Index=$Particles.SpawnGroup, Float Selection Array=DataInterface() {propertyValues {FloatData [1,1,0.3,0.3]}})
    )
    Index 0 Param 4=&RandomRangeFloat(Minimum=0.9, Maximum=2)
    Index 1 Param 1=&RandomRangeFloat(Minimum=1, Maximum=2)
    Index 3 Param 1=&Multiply_Float(
     A=&RandomRangeFloat(Minimum=0.8, Maximum=1.5)
     B=&Lerp_Float(A=1, B=0.5, Alpha=$Particles.Dive)
    )
    Index 3 Param 2=&Multiply_Float(
     A=&SelectFloatFromArray(Direct Array Index=$Particles.SpawnGroup, Float Selection Array=DataInterface() {propertyValues {FloatData [1,1,0,0]}})
     B=&Lerp_Float(A=0, B=90, Alpha=$Particles.Dive)
    )
    Module.Param0WriteEnabled=true
    Module.Param0WriteEnabled1=true
    Module.Param0WriteEnabled3=true
    Module.Param1WriteEnabled=true
    Module.Param1WriteEnabled1=true
    Module.Param1WriteEnabled3=true
    Module.Param2WriteEnabled=true
    Module.Param2WriteEnabled1=true
    Module.Param2WriteEnabled3=true
    Module.Param3WriteEnabled=true
    Module.Param3WriteEnabled1=true
    Module.Param3WriteEnabled3=true
    Write Parameter Index 1=true
    Write Parameter Index 3=true
   )
   &SkeletalMeshLocation(
    Bone / Socket Index=$Particles.SpawnGroup
    Bone Sampling Mode=::NewEnumerator4
    "Interpolate Mesh Sampling (Update)"=true
    MaskSpawnGroup=false
    Orientation Sampling=::NewEnumerator1
    Sampled Orientation Coordinate Space=::Local
    Skeletal Mesh=$System.Skeletal Mesh
   )
   &ScaleMeshSize(Scale Factor=&Multiply_VectorByFloat(
     Vector=<1 1 1>
     Float=&Multiply_Float(A=&FloatFromCurve(FloatCurve=DataInterface() {propertyValues {Curve {keys [<at 0 0>,<at 1 1>]}}}), B=1)
    ))
   &ScaleColor(
    Scale Alpha=&FloatFromCurve(
     FloatCurve=DataInterface() {propertyValues {Curve {keys [<at 0 0>,<at 1 1>]}}}
     Scale Curve=&Multiply_Float(A=0.5, B=$System.SystemFade)
    )
    ScaleA=true
    ScaleRGB=true
   )
   /SetVariables(Particles.MeshOrientation=&LerpQuaternion(
     Quaternion A:Quat4f=$Particles.MeshOrientation
     Quaternion B:Quat4f=&MultiplyQuaternion(
      Quaternion A=&MakeQuaternion(
       Quaternion From=3
       Axis=<1 0 0>
       Angle=0
       From=<0 0 1>
       To=<0 0 1>
       Stabilization Mode:ENiagara_OrientVectorCount=::NewEnumerator3
       Side Direction:Vector3f=<0 1 0>
       Up Direction:Vector3f=<0 0 1>
       Angle Type=::NewEnumerator0
       Coordinate Space=::Local
       Normalize:NiagaraBool=true
      )
      Quaternion B=&MakeQuaternion(
       Quaternion From=0
       Axis=<1 0 0>
       Angle=&Lerp_Float(A=0, B=50, Alpha=$System.SmoothDiveAlpha)
       From=?
       To=?
       Stabilization Mode:ENiagara_OrientVectorCount=::NewEnumerator0
       Side Direction:Vector3f=?
       Up Direction:Vector3f=?
       Angle Type=::NewEnumerator0
       Coordinate Space=::Local
       Normalize:NiagaraBool=true
      )
     )
     Lerp Factor:NiagaraFloat=$System.FE_Lerp
     Enable Spherical Linear Interpolation:NiagaraBool=false
    )) {isSetParameters=true}
  ]
  rendererClasses [&NiagaraMeshRendererProperties]
  renderers [{rendererClass=&NiagaraMeshRendererProperties, rendererIndex=0}]
 }
 Sparks {
  emitterUpdateScript [
   &EmitterState(Life Cycle Mode=::NewEnumerator1, Loop Duration=0.02, UseLoopDelay=false)
   &SpawnRate(SpawnRate=30, Use Spawn Probability=false)
  ]
  particleSpawnScript [
   /SetVariables(Particles.SocketIndex=HlslExpression() {hlslExpression="fmod(Particles.UniqueID,2)"}) {isSetParameters=true}
   &PartitionParticles()
   &SkeletalMeshLocation(
    Bone / Socket Index=&SelectIntFromArray(Direct Array Index=$Particles.PartitionParticles.Partition, Int Selection Array=DataInterface() {propertyValues {IntData [0,1]}})
    Bone Sampling Mode=::NewEnumerator4
    MaskSpawnGroup=false
    Orientation Sampling=::NewEnumerator1
    Sampled Position Offset=&Lerp_Vector(
     A=&SelectVectorFromArray(
      Direct Array Index=$Particles.PartitionParticles.Partition
      Vector Selection Array=DataInterface() {propertyValues {FloatData [<0 0 0>, <0 0 0>]}}
     )
     B=&SelectVectorFromArray(
      Direct Array Index=$Particles.PartitionParticles.Partition
      Vector Selection Array=DataInterface() {propertyValues {FloatData [<0 0 -10>, <0 0 10>]}}
     )
     Alpha=$User.ParaGlideLeanAlpha
    )
    Skeletal Mesh=$System.Skeletal Mesh
   )
   &InitializeParticle(
    Color=<rgba 25 12.5 0 1>
    Lifetime=&RandomRangeFloat(Minimum=1, Maximum=5)
    Position Mode=::NewEnumerator1
    Sprite Size Mode=::NewEnumerator3
    Uniform Sprite Size=&RandomRangeFloat(Minimum=0.5, Maximum=2)
    UsePositionOffset=false
    Write Material Random=true
   )
   &AddVelocity(
    Velocity=&Lerp_Vector(A=<0 0 1>, B=&Lerp_Vector(A=<0 0 1>, B=<0 -1 0.8>, Alpha=$System.SmoothDiveAlpha), Alpha=$System.FE_Lerp)
    Velocity Speed Scale=100
   )
   &ShapeLocation(
    MaskSpawnGroup=false
    Override Seed=false
    Shape Origin=$Particles.SkeletalMeshLocation.SampledPosition
    Sphere Radius=20
    UseOwnerScale=true
   )
  ]
  particleUpdateScript [
   &ParticleState()
   &CameraOffset(Camera Offset Amount=&FloatFromCurve(FloatCurve=DataInterface() {propertyValues {Curve {keys [<at 0 50>,<at 1 -50>]}}}))
   &ScaleSpriteSize(Scale Factor=&Vector2DFromFloat(Value=&FloatFromCurve(
      FloatCurve=DataInterface() {propertyValues {Curve {keys [{arriveTangent=7.4819, interpMode=RCIM_Cubic, leaveTangent=7.4819, tangentMode=RCTM_User, time=0, value=0}, <at 0.0979 0.9367>, <at 1 0>]}}}
      Scale Curve=&RandomRangeFloat(Minimum=1, Maximum=1.2, Evaluation Type=::NewEnumerator1)
     )))
   &ScaleSpriteSize(
    Initial Sprite Size:Vector2f=$Particles.SpriteSize
    Scale Sprite Size Mode=::NewEnumerator0
    Uniform Scale Factor=&RandomRangeFloat(Minimum=0.8, Maximum=1.2, Evaluation Type=::NewEnumerator1)
   ) {module=ScaleSpriteSize001}
   &SolveForcesAndVelocity(Clamp Velocity=false, Limit Acceleration=false)
  ]
  rendererClasses [&NiagaraSpriteRendererProperties]
  renderers [{rendererClass=&NiagaraSpriteRendererProperties}]
 }
 Trail_Chain {
  emitterUpdateScript [
   &EmitterState(
    Inactive Response=::NewEnumerator1
    Life Cycle Mode=::NewEnumerator1
    Loop Behavior=::NewEnumerator1
    Loop Duration Mode=::NewEnumerator1
    UseLoopDelay=false
   )
   &SpawnBurst_Instantaneous(Spawn Count=2, Use Spawn Probability=false, UseLoopCountLimit=false) {module=SpawnBurst_Instantaneous001}
  ]
  particleSpawnScript [
   &PartitionParticles(Number Of Partitions=4)
   &InitializeParticle(
    Color=<rgba 2.5 0.9167 0 1>
    Mesh Scale=&MergeHelper_VectorSwitch(0=<1 1 1>, 1=<1 -1 1>, 2=?, 3=?, Index=$Particles.PartitionParticles.Partition, Num=::NewEnumerator0)
    Mesh Scale Mode=::NewEnumerator4
    UsePositionOffset=false
    Write Material Random=true
   )
   &CameraOffset()
  ]
  particleUpdateScript [
   &ParticleState(Kill Particles When Lifetime Has Elapsed=false, Let Infinitely Lived Particles Die When Emitter Deactivates=true)
   &SkeletalMeshLocation(
    Bone / Socket Index=&SelectIntFromArray(Direct Array Index=$Particles.PartitionParticles.Partition, Int Selection Array=DataInterface() {propertyValues {IntData [0,1,0,1]}})
    Bone Sampling Mode=::NewEnumerator4
    MaskSpawnGroup=false
    Orientation Sampling=::NewEnumerator1
    Sampled Position Offset=&Multiply_Vector(
     A=&MergeHelper_VectorSwitch(0=<-1 -1 -1>, 1=<1 1 1>, 2=?, 3=?, Index=$Particles.PartitionParticles.Partition, Num=::NewEnumerator0)
     B=&FN_FrontEndVector3Modifier(
      Gameplay=&Lerp_Vector(A=<0 0 10>, B=<20 5 10>, Alpha=$User.ParaGlideLeanAlpha)
      Front End=&Lerp_Vector(A=<0 0 0>, B=<5 0 0>, Alpha=$User.ParaGlideLeanAlpha)
     )
    )
    Skeletal Mesh=$System.Skeletal Mesh
   )
   &SkeletalMeshLocation(
    Bone / Socket Index=&SelectIntFromArray(Direct Array Index=$Particles.PartitionParticles.Partition, Int Selection Array=DataInterface() {propertyValues {IntData [0,1,0,1]}})
    Bone Sampling Mode=::NewEnumerator4
    MaskSpawnGroup=false
    Orientation Sampling=::NewEnumerator1
    Sampled Position Offset=&Lerp_Vector(
     A=&SelectVectorFromArray(
      Direct Array Index=$Particles.PartitionParticles.Partition
      Vector Selection Array=DataInterface() {propertyValues {FloatData [<0 0 -10>, <0 0 10>]}}
     )
     B=&SelectVectorFromArray(
      Direct Array Index=$Particles.PartitionParticles.Partition
      Vector Selection Array=DataInterface() {propertyValues {FloatData [<-20 -5 -10>, <20 5 10>]}}
     )
     Alpha=$User.ParaGlideLeanAlpha
    )
    Skeletal Mesh=$System.Skeletal Mesh
   ) {enabled=false, module=SkeletalMeshLocation001}
   &UpdateMeshOrientation(
    Facing Direction=$Engine.Owner.SystemXAxis
    Facing Interpolation=false
    Orientation Method=::NewEnumerator1
    Override Quat=false
    TransformFacing=false
    TransformUp=false
   )
   /SetVariables(Particles.MeshOrientation=&MultiplyQuaternion(
     Quaternion A=$Particles.UpdateMeshOrientation.PreviousFrameQuat
     Quaternion B=&MultiplyQuaternion(
      Quaternion A=&MakeQuaternion(
       Quaternion From=0
       Axis=<0 0 1>
       Angle=90
       From=?
       To=?
       Stabilization Mode:ENiagara_OrientVectorCount=::NewEnumerator0
       Side Direction:Vector3f=?
       Up Direction:Vector3f=?
       Angle Type=::NewEnumerator0
       Normalize:NiagaraBool=true
      )
      Quaternion B=&MakeQuaternion(
       Quaternion From=0
       Axis=<1 0 0>
       Angle=&MergeHelper_FloatSwitch(0=10, 1=-10, 2=?, 3=?, Index=$Particles.PartitionParticles.Partition, Num=::NewEnumerator0)
       From=?
       To=?
       Stabilization Mode:ENiagara_OrientVectorCount=::NewEnumerator0
       Side Direction:Vector3f=?
       Up Direction:Vector3f=?
       Angle Type=::NewEnumerator0
       Normalize:NiagaraBool=true
      )
     )
    )) {isSetParameters=true}
   &DynamicMaterialParameters(
    Index 0 Param 1=&RandomRangeFloat(Minimum=0, Maximum=1)
    Index 0 Param 2=1
    Index 0 Param 3=2
    Index 0 Param 4=1
    Index 1 Param 1=1
    Index 1 Param 2=1
    Index 1 Param 3=1
    Index 1 Param 4=&MakeCustomFloatFromBool(Bool=$User.bIsFrontEndPreview, True Float=0.5, False Float=0.3)
    Module.Param0WriteEnabled=true
    Module.Param0WriteEnabled1=true
    Module.Param1WriteEnabled=true
    Module.Param1WriteEnabled1=true
    Module.Param2WriteEnabled=true
    Module.Param2WriteEnabled1=true
    Module.Param3WriteEnabled=true
    Module.Param3WriteEnabled1=true
    Write Parameter Index 1=true
   )
   &ScaleMeshSize(Scale Factor=&Multiply_VectorByFloat(
     Vector=&MakeVector(
      X=0.6
      Y=1.25
      Z=&Multiply_Float(
       A=&Multiply_Float(
        A=&FloatFromCurve(
         FloatCurve=DataInterface() {propertyValues {Curve {keys [<at 0 1>,<at 1 2>]}}}
         CurveIndex=&Divide_Float(A=&VectorLength(Vector=$Engine.Owner.Velocity), B=9000)
        )
        B=1
       )
       B=1
      )
     )
     Float=$System.SystemFade
    ))
   &InterpolateOverTime(Rate Of Change=40, Target Position Value:NiagaraPosition=$Particles.Position, Variable Type=::NewEnumerator6)
   /SetVariables(Particles.Interpolate Dynamic=&MakeFloat4FromVectorAndFloat(
     "Vector (RGB)"=Linked() {linkedVariable {name=Particles.InterpolateOverTime.InterpolateOverTime_Position.Moving Average, type=NiagaraPosition}}
     "Float (Alpha)"=/Niagara/DynamicInputs/Multiply/ScaleAndBiasFloat(
      Float:NiagaraFloat=0.2
      Scale:NiagaraFloat=&Lerp_Float(A=10, B=5, Alpha=$System.Dive)
      Bias:NiagaraFloat=&Lerp_Float(A=0.15, B=0, Alpha=$System.Dive)
     )
    )) {isSetParameters=true}
   /SetVariables(Particles.MeshOrientation=&LerpQuaternion(
     Quaternion A:Quat4f=$Particles.MeshOrientation
     Quaternion B:Quat4f=&MultiplyQuaternion(
      Quaternion A=&MakeQuaternion(
       Quaternion From=3
       Axis=<1 0 0>
       Angle=0
       From=<0 0 1>
       To=<0 0 1>
       Stabilization Mode:ENiagara_OrientVectorCount=::NewEnumerator3
       Side Direction:Vector3f=<0 0 1>
       Up Direction:Vector3f=<1 0 0>
       Angle Type=::NewEnumerator0
       Coordinate Space=::Local
       Normalize:NiagaraBool=true
      )
      Quaternion B=&MakeQuaternion(
       Quaternion From=0
       Axis=<1 0 0>
       Angle=&Lerp_Float(A=0, B=50, Alpha=$System.SmoothDiveAlpha)
       From=?
       To=?
       Stabilization Mode:ENiagara_OrientVectorCount=::NewEnumerator0
       Side Direction:Vector3f=?
       Up Direction:Vector3f=?
       Angle Type=::NewEnumerator0
       Coordinate Space=::World
       Normalize:NiagaraBool=true
      )
     )
     Lerp Factor:NiagaraFloat=$System.FE_Lerp
     Enable Spherical Linear Interpolation:NiagaraBool=false
    )) {isSetParameters=true}
  ]
  rendererClasses [&NiagaraMeshRendererProperties]
  renderers [{rendererClass=&NiagaraMeshRendererProperties, rendererIndex=0}]
 }
 Trail_Pattern {
  emitterUpdateScript [
   &EmitterState(
    Inactive Response=::NewEnumerator1
    Life Cycle Mode=::NewEnumerator1
    Loop Behavior=::NewEnumerator1
    Loop Duration Mode=::NewEnumerator1
    UseLoopDelay=false
   )
   &SpawnBurst_Instantaneous(Spawn Count=2, Use Spawn Probability=false, UseLoopCountLimit=false)
   /SetVariables(Emitter.TrailTex_001SpeedU:NiagaraFloat=&MakeCustomFloatFromBool(Bool=&SetBoolByFloatComparison(A=$User.ParaGlideLeanAlpha, B=0.2), True Float=-2.5, False Float=-1.25)) {isSetParameters=true}
  ]
  particleSpawnScript [
   &PartitionParticles()
   &InitializeParticle(Color Mode=::NewEnumerator0, Mesh Scale Mode=::NewEnumerator4, UsePositionOffset=false, Write Material Random=true)
   &InitialMeshOrientation(Mesh Orientation Mode=::NewEnumerator2, Rotation=<0 -0.25 0>, Use Rotation Vector=true)
   &InitialMeshOrientation(
    Mesh Orientation Mode=::NewEnumerator2
    Rotation=&SelectVectorFromArray(
     Direct Array Index=$Particles.PartitionParticles.Partition
     Vector Selection Array=DataInterface() {propertyValues {FloatData [<0 0 0>, <0 0 0.5>]}}
    )
    Use Rotation Vector=true
   ) {module=InitialMeshOrientation001}
  ]
  particleUpdateScript [
   &ParticleState(Kill Particles When Lifetime Has Elapsed=false, Let Infinitely Lived Particles Die When Emitter Deactivates=true)
   &ScaleMeshSize(Scale Factor=&MakeCustomVectorFromBool(
     Bool=$User.bisFrontEndPreview
     True Vector=&Lerp_Vector(A=<0.7 1 1>, B=<1 1 1>, Alpha=$System.InterpolateOverTime.InterpolateOverTime_Float.Moving Average)
     False Vector=&Lerp_Vector(A=<0.9 1 1>, B=<1.2 1 1>, Alpha=$System.InterpolateOverTime.InterpolateOverTime_Float.Moving Average)
    ))
   &SkeletalMeshLocation(
    Bone / Socket Index=&SelectIntFromArray(Direct Array Index=$Particles.PartitionParticles.Partition, Int Selection Array=DataInterface() {propertyValues {IntData [0,1]}})
    Bone Sampling Mode=::NewEnumerator4
    MaskSpawnGroup=false
    Orientation Sampling=::NewEnumerator1
    Sampled Position Offset=&Lerp_Vector(
     A=&SelectVectorFromArray(
      Direct Array Index=$Particles.PartitionParticles.Partition
      Vector Selection Array=DataInterface() {propertyValues {FloatData [<0 0 0>, <0 0 0>]}}
     )
     B=&SelectVectorFromArray(
      Direct Array Index=$Particles.PartitionParticles.Partition
      Vector Selection Array=DataInterface() {propertyValues {FloatData [<0 0 -10>, <0 0 10>]}}
     )
     Alpha=$User.ParaGlideLeanAlpha
    )
    Skeletal Mesh=$System.Skeletal Mesh
   )
   &ScaleColor(
    Scale Alpha=&Multiply_Float(A=0.75, B=$System.SystemFade)
    Scale RGB=&MakeVectorFromLinearColorRGB(Color=<rgba 10 6.0145 2 0.75>)
    ScaleA=true
    ScaleRGB=true
   )
   /SetVariables(Particles.ColorGlow:LinearColor=&MakeLinearColorFromVectorAndFloat(
     "Vector (RGB)"=&MakeVectorFromLinearColorRGB(Color=<rgba 1.5 1.5 1.5 1>)
     "Float (Alpha)"=&Multiply_Float(
      A=&MakeCustomFloatFromBool(Bool=$User.bisFrontEndPreview, True Float=1, False Float=2)
      B=$System.SystemFade
     )
    )) {isSetParameters=true}
   &DynamicMaterialParameters(
    Index 2 Param 1=1
    Index 2 Param 3=&MakeCustomFloatFromBool(Bool=$User.bisFrontEndPreview, True Float=0.7, False Float=1)
    Module.Index2Float0Write=true
    Module.Index2Float1Write=true
    Module.Index2Float2Write=true
    Module.Index2Float3Write=false
    Write Parameter Index 0=::NewEnumerator0
    Write Parameter Index 2=::NewEnumerator1
   )
   /SetVariables(Particles.MeshOrientation=&LerpQuaternion(
     Quaternion A:Quat4f=$Particles.MeshOrientation
     Quaternion B:Quat4f=&MultiplyQuaternion(
      Quaternion A=&MakeQuaternion(
       Quaternion From=3
       Axis=<1 0 0>
       Angle=0
       From=<1 0 0>
       To=<0 0 1>
       Stabilization Mode:ENiagara_OrientVectorCount=::NewEnumerator3
       Side Direction:Vector3f=<0 0 1>
       Up Direction:Vector3f=<1 0 0>
       Angle Type=::NewEnumerator0
       Coordinate Space=::Local
       Normalize:NiagaraBool=true
      )
      Quaternion B=&MakeQuaternion(
       Quaternion From=0
       Axis=<0 1 0>
       Angle=&Lerp_Float(A=0, B=-50, Alpha=$System.SmoothDiveAlpha)
       From=?
       To=?
       Stabilization Mode:ENiagara_OrientVectorCount=::NewEnumerator0
       Side Direction:Vector3f=?
       Up Direction:Vector3f=?
       Angle Type=::NewEnumerator0
       Coordinate Space=::World
       Normalize:NiagaraBool=true
      )
     )
     Lerp Factor:NiagaraFloat=$System.FE_Lerp
     Enable Spherical Linear Interpolation:NiagaraBool=false
    )) {isSetParameters=true}
  ]
  rendererClasses [&NiagaraMeshRendererProperties]
  renderers [
   {rendererClass=&NiagaraMeshRendererProperties, rendererIndex=0}
   {rendererClass=&NiagaraMeshRendererProperties, rendererIndex=1}
  ]
 }
 Trail_Soft {
  emitterUpdateScript [
   &EmitterState(
    Inactive Response=::NewEnumerator1
    Life Cycle Mode=::NewEnumerator1
    Loop Behavior=::NewEnumerator1
    Loop Duration Mode=::NewEnumerator1
    UseLoopDelay=false
   )
   &SpawnBurst_Instantaneous(Spawn Count=2, Use Spawn Probability=false, UseLoopCountLimit=false)
   /SetVariables(Emitter.Texture_Speed_Master:NiagaraFloat=&MakeCustomFloatFromBool(
     Bool=&SetBoolByFloatComparison(A=1, B=$User.ParaGlideLeanAlpha, Comparison Type=::NewEnumerator4)
     True Float=0.7
     False Float=&RandomRangeFloat(Minimum=0.4, Maximum=0.5)
    )) {isSetParameters=true}
   /SetVariables(Emitter.Texture_Speed_Master_Detail:NiagaraFloat=&MakeCustomFloatFromBool(
     Bool=&SetBoolByFloatComparison(A=1, B=$User.ParaGlideLeanAlpha, Comparison Type=::NewEnumerator4)
     True Float=0.6
     False Float=0.45
    )) {isSetParameters=true}
   /SetVariables(Emitter.SphereMaskU:NiagaraFloat=&Lerp_Float(A=1, B=0.95, Alpha=$User.ParaGlideLeanAlpha)) {isSetParameters=true}
  ]
  particleSpawnScript [
   &PartitionParticles()
   &InitializeParticle(
    Color=<rgba 0.4844 0.2302 0.0315 1>
    Mesh Scale=&SelectVectorFromArray(
     Direct Array Index=$Particles.PartitionParticles.Partition
     Vector Selection Array=DataInterface() {propertyValues {FloatData [<1 1 1>, <1 1 1>]}}
    )
    Mesh Scale Mode=::NewEnumerator4
    UsePositionOffset=true
    Write Material Random=true
   )
   &InitialMeshOrientation(Mesh Orientation Mode=::NewEnumerator2, Rotation=<0.25 0 0>, Use Rotation Vector=true)
   &InitialMeshOrientation(
    Mesh Orientation Mode=::NewEnumerator2
    Rotation=&SelectVectorFromArray(
     Direct Array Index=$Particles.PartitionParticles.Partition
     Vector Selection Array=DataInterface() {propertyValues {FloatData [<0 0 0>, <0 0 0.1>]}}
    )
    Use Rotation Vector=true
   ) {module=InitialMeshOrientation001}
  ]
  particleUpdateScript [
   &ParticleState(Kill Particles When Lifetime Has Elapsed=false, Let Infinitely Lived Particles Die When Emitter Deactivates=true)
   &ScaleMeshSize(Scale Factor=&MakeCustomVectorFromBool(
     Bool=$User.bIsFrontEndPreview
     True Vector=&Lerp_Vector(A=<1 0.6 1>, B=<1 1.3 1>, Alpha=$User.ParaGlideLeanAlpha)
     False Vector=&Lerp_Vector(A=<1 0.9 1>, B=<1 1.5 1>, Alpha=$User.ParaGlideLeanAlpha)
    ))
   &SkeletalMeshLocation(
    Bone / Socket Index=&SelectIntFromArray(Direct Array Index=$Particles.PartitionParticles.Partition, Int Selection Array=DataInterface() {propertyValues {IntData [0,1]}})
    Bone Sampling Mode=::NewEnumerator4
    MaskSpawnGroup=false
    Orientation Sampling=::NewEnumerator1
    Sampled Position Offset=&Lerp_Vector(
     A=&SelectVectorFromArray(
      Direct Array Index=$Particles.PartitionParticles.Partition
      Vector Selection Array=DataInterface() {propertyValues {FloatData [<0 -10 0>, <0 10 0>]}}
     )
     B=&SelectVectorFromArray(
      Direct Array Index=$Particles.PartitionParticles.Partition
      Vector Selection Array=DataInterface() {propertyValues {FloatData [<0 0 -10>, <0 0 10>]}}
     )
     Alpha=$User.ParaGlideLeanAlpha
    )
    Skeletal Mesh=$System.Skeletal Mesh
   )
   &DynamicMaterialParameters(
    Index 0 Param 1=&RandomRangeFloat(Minimum=0.2, Maximum=1)
    Index 0 Param 2=&RandomRangeFloat(Minimum=0.2, Maximum=1)
    Index 0 Param 3=&FN_FrontEndFloatModifier(Gameplay=0.5, Front End=0)
    Index 0 Param 4=20
    Index 3 Param 1=150
    Index 3 Param 2=0.3
    Index 3 Param 4=&FloatFromCurve(FloatCurve=DataInterface() {propertyValues {Curve {keys [<at 0 0>,<at 1 1>]}}})
    Module.Index0Param0Write=true
    Module.Index0Param1Write=true
    Module.Index0Param2Write=true
    Module.Index0Param3Write=true
    Module.Index2Param0Write=true
    Module.Index2Param1Write=true
    Module.Index2Param2Write=true
    Module.Index2Param3Write=true
    Module.Index3Param0Write=true
    Module.Index3Param1Write=true
    Module.Index3Param2Write=true
    Module.Index3Param3Write=true
    Write Parameter Index 2=true
    Write Parameter Index 3=true
   )
   &InterpolateOverTime(Rate Of Change=10, Target Vector Value:NiagaraPosition=$Particles.Position, Variable Type=::NewEnumerator2)
   /SetVariables(Particles.Interpolate Dynamic=&MakeFloat4FromVectorAndFloat(
     "Vector (RGB)"=$StackContext.InterpolateOverTime.InterpolateOverTime_V3_.Moving Average
     "Float (Alpha)"=&FN_FrontEndFloatModifier(Gameplay=0, Front End=0)
    )) {isSetParameters=true}
   /SetVariables(Particles.Color002=&Multiply_LinearColorByFloat(
     Linear Color=&MultiplyLinearColor(
      Linear Color=<rgba 1 0.3085 0.1042 1>
      Scale Factor=&MakeVector4_FromLinearColor(LinearColor=&FNTimeOfDayColor(Night Color=<rgba 0.1927 0.3945 1 1>, Morning Color=<rgba 0.6 0.6 0.6 1>, Day Color=<rgba 0.8 0.8 0.8 1>, Evening Color=<rgba 0.6 0.6 0.6 1>))
     )
     Float=&Multiply_Float(
      A=&Multiply_Float(
       A=&FN_FrontEndFloatModifier(Gameplay=0.6, Front End=0.6)
       B=&FN_PickFloatDependOnPlatform(TargetPlatformActive=1, TargetPlatformNoActive=0.7)
      )
      B=$System.SystemFade
     )
    )) {isSetParameters=true}
   /SetVariables(Particles.Size_Detail:Vector3f=&MakeCustomVectorFromBool(
     Bool=$User.bIsFrontEndPreview
     True Vector=&Lerp_Vector(A=<1 0.7 1>, B=<1 1.2 1>, Alpha=$User.ParaGlideLeanAlpha)
     False Vector=&Lerp_Vector(A=<1 0.8 1>, B=<1 1.5 1>, Alpha=$User.ParaGlideLeanAlpha)
    )) {isSetParameters=true}
   &ScaleColor(
    Scale Alpha=&FloatFromCurve(
     FloatCurve=DataInterface() {propertyValues {Curve {keys [<at 0 0>,<at 0.1 1>]}}}
     Scale Curve=&Multiply_Float(
      A=&Multiply_Float(
       A=&FN_FrontEndFloatModifier(Gameplay=1, Front End=0.8)
       B=&FN_PickFloatDependOnPlatform(TargetPlatformActive=1, TargetPlatformNoActive=0.8)
      )
      B=$System.SystemFade
     )
    )
    Scale RGB=&Multiply_Vector(
     A=&MakeVectorFromLinearColorRGB(Color=&FNTimeOfDayColor(Night Color=<rgba 0.5 0.625 1 1>, Morning Color=<rgba 1 1 1 1>, Day Color=<rgba 1 1 1 1>, Evening Color=<rgba 1 1 1 1>))
     B=&VectorFromFloat(Value=&FN_FrontEndFloatModifier(Gameplay=1, Front End=0.3))
    )
    ScaleA=true
    ScaleRGB=true
   )
   /SetVariables(Particles.MeshOrientation=&LerpQuaternion(
     Quaternion A:Quat4f=$Particles.MeshOrientation
     Quaternion B:Quat4f=&MultiplyQuaternion(
      Quaternion A=&MakeQuaternion(
       Quaternion From=3
       Axis=<1 0 0>
       Angle=0
       From=<0 1 0>
       To=<0 0 1>
       Stabilization Mode:ENiagara_OrientVectorCount=::NewEnumerator3
       Side Direction:Vector3f=<0 0 1>
       Up Direction:Vector3f=<0 1 0>
       Angle Type=::NewEnumerator0
       Coordinate Space=::Local
       Normalize:NiagaraBool=true
      )
      Quaternion B=&MakeQuaternion(
       Quaternion From=0
       Axis=<0 0 1>
       Angle=&Lerp_Float(A=0, B=-50, Alpha=$System.SmoothDiveAlpha)
       From=?
       To=?
       Stabilization Mode:ENiagara_OrientVectorCount=::NewEnumerator0
       Side Direction:Vector3f=?
       Up Direction:Vector3f=?
       Angle Type=::NewEnumerator0
       Coordinate Space=::World
       Normalize:NiagaraBool=true
      )
     )
     Lerp Factor:NiagaraFloat=$System.FE_Lerp
     Enable Spherical Linear Interpolation:NiagaraBool=false
    )) {isSetParameters=true}
  ]
  rendererClasses [&NiagaraMeshRendererProperties]
  renderers [
   {rendererClass=&NiagaraMeshRendererProperties, rendererIndex=0}
   {rendererClass=&NiagaraMeshRendererProperties, rendererIndex=1}
  ]
 }
}
summary {
 systemName=NS_Contrail_RoseScour
 emitters [
  {emitterName=Trail_Soft, rendererClasses [&NiagaraMeshRendererProperties]}
  {emitterName=Trail_Chain, rendererClasses [&NiagaraMeshRendererProperties]}
  {emitterName=Sparks, rendererClasses [&NiagaraSpriteRendererProperties]}
  {emitterName=Trail_Pattern, rendererClasses [&NiagaraMeshRendererProperties]}
  {emitterName=Red_Trail, rendererClasses [&NiagaraMeshRendererProperties]}
  {
   emitterName=Hand_Feet_Flares
   rendererClasses [&NiagaraLightRendererProperties, &NiagaraSpriteRendererProperties]
  }
 ]
 userVariables [
  User.bIsFrontEnd {type=NiagaraBool}
  User.bIsFrontEndPreview {type=NiagaraBool}
  User.ParaGlideLeanAlpha {type=NiagaraFloat, defaultValue=0, description=""}
 ]
}
